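; BP_SceneCaptureClyde is a UI-related actor, by name a scene capture for Clyde, that parks itself high above the level.
; Its construction script reads the world origin location, keeps X and Y as floats, and sets the actor's location at a very large fixed Z, without sweep or teleport.
; BeginPlay, ActorBeginOverlap and Tick are empty, so the graph holds no runtime logic.
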(event EventBeginPlay)

(event Collision|EventActorBeginOverlap
  (params (OtherActor :type Actor%20Object%20Reference)))

(event EventTick
  (params (DeltaSeconds :type Float%20%28single-precision%29)))

(fn ConstructionScript
  (bind (_x _y) (Utilities|Struct|BreakIntVector (Game|GetWorldOriginLocation)))
  (Transformation|SetActorLocation 0 (Math|Vector|MakeVector (Math|Conversions|ToFloat(Integer) _x) (Math|Conversions|ToFloat(Integer) _y) 1e+05) false false))
