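; Manager Blueprint that overrides the minimap location entries and name fonts for three Clyde POIs: LemonCart, BerryTart and ClayPlug.
; On BeginPlay, only the authority waits for playlist data to be ready, then delays briefly before updating each POI from its tag and font variables.
; Each update passes the tag to the GameState's GSC_MiniMapDataOverride component as both OverrideLocations containers, then sets that location's font in all minimap states.
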(event EventBeginPlay
  (Networking|SwitchHasAuthority
    (:Authority
      (Class|FortAsyncActionCurrentPlaylistReady|PlaylistDataReadyAsync
        (:OnReady
          (Utilities|FlowControl|Delay 2.0)
          (CallFunction|UpdateClydeMinimapNameFonts))))))

(fn SetMiniMapOverride
  (params (SelectedPOITag :type Gameplay%20Tag%20Structure) (Font :type Slate%20Font%20Info%20Structure%20%28by%20ref%29))
  (LocalVariables|SetMiniMapOverrideComponent (Actor|GetComponentbyClass (Fortnite|Athena|GameState|GetGameStateAthena) "/Game/Blueprints/Components/GSC_MiniMapDataOverride.GSC_MiniMapDataOverride_C"))
  (Class|FortGameStateComponentMiniMapDataOverride|OverrideLocations (LocalVariables|GetMiniMapOverrideComponent) (GameplayTags|MakeGameplayTagContainerfromTag SelectedPOITag) (GameplayTags|MakeGameplayTagContainerfromTag SelectedPOITag))
  (UI|Minimap|SetLocationFontinAllStates (Fortnite|Athena|GameState|GetGameStateAthena) SelectedPOITag Font))

(fn UpdateClydeMinimapNameFonts
  (CallFunction|SetMiniMapOverride (Variables|LemonCart|GetLemonCartPOITag) (Variables|LemonCart|GetLemonCartFont))
  (CallFunction|SetMiniMapOverride (Variables|BerryTart|GetBerryTartPOITag) (Variables|BerryTart|GetBerryTartFont))
  (CallFunction|SetMiniMapOverride (Variables|ClayPlug|GetClayPlugPOITag) (Variables|ClayPlug|GetClayPlugFont)))

(fn ConstructionScript)
