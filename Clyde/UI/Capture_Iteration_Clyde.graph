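# Editor utility task for capturing minimap textures of the Clyde Apollo_Terrain_Retro map.
# It runs two Capture_Minimap child tasks in sequence, each loading everything at a fixed high resolution while preserving existing state.
# The first uses the discovered scene capture to target the discovered minimap texture; on completion, the second does the same with the undiscovered capture and texture.
# The task finishes once the second capture completes.
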EditorUtilityBlueprint /Clyde/UI/Capture_Iteration_Clyde
graphs [EventGraph {nodes [
   Event BeginExecution()
   > n1 = Construct Object from Class!(
    execute
    Class=/Game/Athena/Editor/Capture_Minimap.Capture_Minimap_C
    self=Self-Reference()
    LoadEverything:bool=True
    PreserveExistingState:bool=true
    LoadMapOnly:bool=False
    Resolution:int=4096
    Map:Soft Object Path Structure=Get Map() -> (Map:Soft Object Path Structure)
    ImportTargets:Array of Texture 2D Object References=Get Discovered Import Targets() -> (Discovered Import Targets:Array of Texture 2D Object References)
    Scene Capture Class:Scene Capture 2D Class Reference=/Game/Athena/Editor/SceneCapture_Discovered.SceneCapture_Discovered_C
   ) -> (then, ReturnValue:Capture Minimap Object Reference)
   > n2 = RegisterAndExecuteTask!(execute:Exec, Task:Editor Utility Task Object Reference=@n1, OptionalParentTask:Editor Utility Task Object Reference=Self-Reference()) -> (then:Exec, OnFinished:Exec, Task:Editor Utility Task Object Reference) {is=title K2Node_AsyncAction}
   n3 = Construct Object from Class(
    execute=@n2.OnFinished
    Class=&Capture_Minimap_C
    self=Self-Reference()
    LoadEverything=True
    PreserveExistingState=true
    LoadMapOnly=False
    Resolution=4096
    Map=Get Map()
    ImportTargets=Get Undiscovered Import Targets() -> (Undiscovered Import Targets:Array of Texture 2D Object References)
    Scene Capture Class=/Game/Athena/Editor/SceneCapture_Undiscovered.SceneCapture_Undiscovered_C
   )
   > n4 = RegisterAndExecuteTask(Task=@n3, OptionalParentTask=Self-Reference())
   FinishExecutingTask(execute=@n4.OnFinished)
  ]}]
variables [
 Discovered Import Targets {type="array<Texture2D>", default [/Clyde/UI/Textures/S2/Apollo_Terrain_MinimapS2]}
 Map {type=SoftObjectPath, default=/Clyde/Apollo_Terrain_Retro}
 Undiscovered Import Targets {type="array<Texture2D>", default [/Clyde/UI/Textures/S2/Apollo_Terrain_Minimap_UndiscoveredS2]}
]
