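# Clyde variant of the Athena inventory group panel widget, stacking the Ingredients, Resources and Ammo item groups under a "Resources" header.
# Construct, and PreConstruct outside design time, fill InventoryGroupArray with those three group widgets in that order.
# GetOrCreateInventoryGroups returns that array, building it lazily with the same three widgets if it is still empty.
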WidgetBlueprint /Clyde/UI/Clyde_AthenaInventoryGroupPanel
graphs [
 EventGraph {nodes [
   n1 = Event PreConstruct()
   > n2 = Branch(Condition=@n1.IsDesignTime)
   Set InventoryGroupArray(
    execute=[@n3.then, @n2.else]
    InventoryGroupArray:Array of Athena Inventory Group Base Object References=Make Array!(
     "[0]":Athena Inventory Group Base Object Reference=Get AthenaInventoryGroupIngredients() -> (AthenaInventoryGroupIngredients:Athena Inventory Group Ingredients Object Reference)
     "[1]":Athena Inventory Group Base Object Reference=Get AthenaInventoryGroupResources() -> (AthenaInventoryGroupResources:Athena Inventory Group Resources Object Reference)
     "[2]":Athena Inventory Group Base Object Reference=Get AthenaInventoryGroupAmmo() -> (AthenaInventoryGroupAmmo:Athena Inventory Group Ammo Object Reference)
    ) -> (Array:Array of Athena Inventory Group Base Object References)
   ) -> (Output_Get:Array of Athena Inventory Group Base Object References)
   n3 = Event Construct()
   Event Tick()
  ]}
 GetOrCreateInventoryGroups {nodes [
   GetOrCreateInventoryGroups() {is=title K2Node_FunctionEntry}
   n1 = Get InventoryGroupArray() -> (InventoryGroupArray:Array of Athena Inventory Group Base Object References)
   > n2 = Branch(Condition="integer > integer"(A=Length(TargetArray:Array of Athena Inventory Group Base Object References=@n1)))
   Return Node!(execute=[@n2.then, @n3.then], OutInventoryGroups:Array of Athena Inventory Group Base Object References=Get InventoryGroupArray())
   n3 = Set InventoryGroupArray(
    execute=@n2.else
    InventoryGroupArray=Make Array("[0]"=Get AthenaInventoryGroupIngredients(), "[1]"=Get AthenaInventoryGroupResources(), "[2]"=Get AthenaInventoryGroupAmmo())
   )
  ]}
]
variables [InventoryGroupArray {type="array<AthenaInventoryGroupBase>", default []}]
widgets {
 info {inheritedWidgetCount=1, parentClass=/Script/FortniteUI.AthenaInventoryGroupPanel, rootWidgetClass=&Overlay, widgetCount=7}
 widgets [
  Overlay_248 {widgetClassPath=&Overlay, widget=Overlay_248}
  Resources {
   widgetClassPath=&VerticalBox
   parent=Overlay_248
   widget=Resources
   slot {class=OverlaySlot, properties {padding=<ltrb 0 100 0 0>}}
  }
  CommonTextBlock_1 {
   widgetClassPath=&CommonTextBlock
   parent=Resources
   widget=CommonTextBlock_1
   properties {
    lineHeightPercentage=0.9
    mobileFontSizeMultiplier=1.2
    scrollStyle=&TextScrollStyle-Base_C
    text=Resources
    textTransformPolicy=ToUpper
    font {
     bForceMonospaced=false
     bMaterialIsStencil=false
     fontMaterial=/Game/Athena/HUD/Inventory/Art/M_UI_Inventory_CatergoryHeader_FontColor
     fontObject=&BurbankSmall
     letterSpacing=0
     monospacedWidth=1
     size=13
     skewAmount=0
     typefaceFontName=Black
     outlineSettings {
      bApplyOutlineToDropShadows=false
      bMiteredCorners=false
      bSeparateFillAlpha=true
      outlineMaterial=&M-FontShadow
      outlineSize=2
      outlineColor=<rgba 0 0 0 0.5>
     }
    }
    margin=<ltrb 0 -2 0 0>
    renderTransform {angle=0, scale=<1 1>, shear=<-12 0>, translation=<0 0>}
    shadowOffset=<0 0>
   }
   slot {class=VerticalBoxSlot, properties {verticalAlignment=VAlign_Top}}
  }
  AthenaInventoryGroupIngredients {
   widgetClassPath=/Game/Athena/HUD/Inventory/AthenaInventoryGroupIngredients
   bIsVariable=true
   parent=Resources
   widget=AthenaInventoryGroupIngredients
   slot {class=VerticalBoxSlot, properties {padding=<ltrb 0 3 0 3>}}
  }
  AthenaInventoryGroupResources {
   widgetClassPath=/Game/Athena/HUD/Inventory/AthenaInventoryGroupResources
   bIsVariable=true
   parent=Resources
   widget=AthenaInventoryGroupResources
   slot {class=VerticalBoxSlot, properties {padding=<ltrb 0 3 0 3>}}
  }
  AthenaInventoryGroupAmmo {
   widgetClassPath=/Game/Athena/HUD/Inventory/AthenaInventoryGroupAmmo
   bIsVariable=true
   parent=Resources
   widget=AthenaInventoryGroupAmmo
   slot {class=VerticalBoxSlot, properties {padding=<ltrb 0 3 0 3>}}
  }
  ExtensionRowHolder {widgetClassPath=&VerticalBox, bInherited=true, widget=null}
 ]
}
