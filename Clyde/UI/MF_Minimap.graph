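# MF_Minimap is a UI material function that supplies the color and alpha of a minimap.
# It samples a map texture for the RGB output and the red channel of a separate grayscale opacity texture for Alpha.
# Both samples share an overridable UV input that falls back to texture coordinates.
# The texture inputs are also optional, defaulting to the Apollo terrain minimap and an engine white grid texture.
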MaterialFunction /Clyde/UI/MF_Minimap
graph {nodes [
  n1 = FunctionInput(Preview=TextureCoordinate()) {bUsePreviewValueAsDefault=true, inputName=UVsOverride, inputType=FunctionInput_Vector2}
  n2 = TextureObject() {texture=/Clyde/UI/Textures/Apollo_Terrain_Minimap}
  FunctionOutput(None=TextureSample(
    UVs=@n1
    Tex=FunctionInput(Preview=@n2) {bUsePreviewValueAsDefault=true, inputName=MapTexture, inputType=FunctionInput_Texture2D, sortPriority=1}
   ) {texture=&Apollo_Terrain_Minimap}.RGB) {outputName=RGB}
  n3 = TextureObject() {samplerType=SAMPLERTYPE_Grayscale, texture=/Engine/EngineMaterials/DefaultWhiteGrid_Low}
  FunctionOutput(None=TextureSample(UVs=@n1, Tex=FunctionInput(Preview=@n3) {bUsePreviewValueAsDefault=true, inputName=MapOpacityTexture, inputType=FunctionInput_Texture2D}) {samplerType=SAMPLERTYPE_Grayscale, texture=&Apollo_Terrain_Minimap}.R) {outputName=Alpha, sortPriority=1}
 ]}
