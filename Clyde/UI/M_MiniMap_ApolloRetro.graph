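# Translucent UI material for a minimap, defaulting to the Apollo terrain minimap texture and a post mask, with optional point-of-interest art on top.
# MF_Minimap supplies the map color from the map and opacity textures.
# POI UVs are offset by a position parameter and scaled by size, widget pixel size and the smaller screen dimension.
# A box mask clips the icon, which alpha-blends over the map behind the "Use POI Art" static switch.
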Material /Clyde/UI/M_MiniMap_ApolloRetro
graph {
 properties {blendMode=BLEND_Translucent, materialDomain=MD_UI, pixelDepthOffsetMode=PDOM_Legacy, usedShadingModels=Default Lit}
 nodes [
  n1 = MaterialFunctionCall() {materialFunction=&GetUserInterfaceUV}
  n2 = MaterialFunctionCall() {materialFunction=&ScreenResolution}
  n3 = Divide(A=@n1.Pixel Size, B=Min(A=ComponentMask(None=@n2.Visible Resolution) {r=true}, B=ComponentMask(None=@n2.Visible Resolution) {g=true}))
  n4 = ScalarParameter() {defaultValue=0.15, parameterName=POI Size}
  n5 = Divide(A=@n4)
  n6 = Subtract(A=ComponentMask(None=VectorParameter() {parameterName=POI Position, defaultValue=<rgba -0.105 -0.045 0 1>}.RGB) {g=true, r=true}, B=@n5)
  n7 = MaterialFunctionCall!(
   Texture Scale=Multiply(A=Divide(B=@n3), B=@n4)
   UVs=Subtract(A=@n1.Normalized UV, B=Add(A=@n6, B=@n5))
   Offset XY Texturesize Z=AppendVector(A=Constant2Vector() {g=0.49, r=0.5}, B=Constant() {r=1})
  ) {materialFunction=/Engine/Functions/Engine_MaterialFunctions01/Texturing/ScaleUVsAroundPoint}
  n8 = Saturate(None=@n7.UVs)
  AppendVector(A=@n8, B=Constant())
  n9 = Abs(None=ConstantBiasScale(None=@n7.UVs) {bias=-0.5, scale=2})
  OneMinus(None=Saturate(None=Floor(None=@n9)))
  n10 = MaterialFunctionCall!(
   UVsOverride
   MapOpacityTexture=TextureObjectParameter() {parameterName=MapOpacityTexture, samplerType=SAMPLERTYPE_Grayscale, texture=/Clyde/UI/Textures/T_Clyde_Minimap_PostMask}
   MapTexture=TextureObjectParameter() {parameterName=MapTexture, texture=/Clyde/UI/Textures/Apollo_Terrain_Minimap}
  ) {materialFunction=&/MF_Minimap}
  n11 = TextureSample(UVs=@n8, Tex=TextureObjectParameter() {parameterName=POI 1, texture=/Game/Athena/UI/InGame/Sidekick/Reactions/100})
  n12 = Multiply(
   A=@n11.A
   B=MaterialFunctionCall!(A=@n7.UVs, B=Constant2Vector() {g=0.5, r=0.5}, Bounds=Constant() {r=1}, Edge Falloff=Constant()) {materialFunction=/Engine/Functions/Engine_MaterialFunctions01/Gradient/BoxMask-2D}.Result
  )
  n13 = OneMinus(None=@n12)
  StaticSwitchParameter(True=Add(A=Multiply(A=@n10.RGB, B=@n13), B=Multiply(A=@n11.RGB, B=@n12)), False=@n10.RGB) {parameterName=Use POI Art}
  TextureObject()
 ]
}
