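# Material function that turns a minimap texture, by default the Apollo terrain map, into a stylized blue version, apparently the fog-of-war look.
# It gamma-adjusts, posterizes and desaturates the sampled color, then a SmoothCurve, scaled to half range, blends between dark and light blue.
# A SphereMask on RGB distance color-keys pixels near a dark navy, likely water, to a flat blue.
# No visibility mask is read; only texture and UVs are inputs.
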MaterialFunction /Clyde/UI/MF_Minimap_FOW
graph {nodes [
  n1 = Constant3Vector() {constant=<rgba 0.0269 0.0674 0.1024 1>}
  n2 = Constant3Vector() {constant=<rgba 0.2541 0.4345 0.5608 1>}
  n3 = TextureSample(
   UVs=FunctionInput(Preview=TextureCoordinate()) {bUsePreviewValueAsDefault=true, inputName=UVsOverride, inputType=FunctionInput_Vector2}
   Tex=FunctionInput(Preview=TextureObject() {texture=/Clyde/UI/Textures/Apollo_Terrain_Minimap}) {bUsePreviewValueAsDefault=true, inputName=MapTexture, inputType=FunctionInput_Texture2D}
  )
  n4 = Constant() {r=18}
  n5 = Multiply(A=Power(Base=@n3.RGB, Exp=Constant() {r=0.75}), B=@n4)
  n6 = Desaturation(None=Divide(A=Floor(None=@n5), B=@n4))
  n7 = Constant() {r=2.1}
  n8 = Constant() {r=1.3}
  n9 = Saturate(None=LinearInterpolate(
    A=Constant()
    B=Constant() {r=0.5}
    Alpha=MaterialFunctionCall!(X=@n6, Tangent 0=@n7, Tangent 1=@n8) {materialFunction=/Engine/Functions/Engine_MaterialFunctions02/Gradients/SmoothCurve}.Result
   ))
  n10 = Constant3Vector() {constant=<rgba 0.0024 0.0194 0.0578 1>}
  n11 = Constant() {r=0.05}
  n12 = Constant()
  FunctionOutput(None=LinearInterpolate(
    A=LinearInterpolate(A=@n1, B=@n2, Alpha=@n9)
    B=Constant3Vector() {constant=<rgba 0.0338 0.1341 0.2344 1>}
    Alpha=SphereMask(A=@n3.RGB, B=@n10, Radius=@n11, Hardness=@n12)
   )) {outputName=RGB}
 ]}
