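# Translucent UI material for the Apollo minimap that blends the normal map with a fog-of-war version.
# The blend mask combines a discoverability texture with the fog render target, whose G channel is gated by a circle growing from CurrentStartLocation with DiscoverRevealProgress from a parameter collection.
# Panning Perlin noise distorts fog UVs along the reveal edge, where a tinted line is added.
# POI bitmask unpack branches do not feed that blend.
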Material /Clyde/UI/M_MinimapApolloRetro_FOW
graph {
 properties {blendMode=BLEND_Translucent, materialDomain=MD_UI, pixelDepthOffsetMode=PDOM_Legacy, usedShadingModels=Default Lit}
 nodes [
  n1 = TextureObjectParameter() {parameterName=MapOpacityTexture, samplerType=SAMPLERTYPE_Grayscale, texture=/Clyde/UI/Textures/T_Clyde_Minimap_PostMask}
  n2 = TextureObjectParameter() {parameterName=MapTexture, texture=/Clyde/UI/Textures/Apollo_Terrain_Minimap}
  n3 = TextureObjectParameter() {parameterName=DiscoverabilityTexture, samplerType=SAMPLERTYPE_LinearGrayscale, texture=/Clyde/UI/Textures/Clyde_Stencil_Mask_S1_}
  n4 = Constant() {r=1}
  n5 = If(
   A=Multiply(A=TextureSample(Tex=@n3) {samplerType=SAMPLERTYPE_LinearGrayscale, texture=/Clyde/UI/Textures/Apollo_Terrain_Minimap_Discoverabilty}.R) {constB=255}
   B=Constant() {r=255}
   "A > B"=@n4
   "A == B"=Constant()
   "A < B"=@n4
  )
  n6 = TextureSample() {texture=/Game/Athena/Apollo/Maps/UI/RT_FogofWar}
  n7 = CollectionParameter() {collection=/Game/Athena/Apollo/Maps/UI/DiscoverabilityMPC, parameterName=CurrentStartLocation}
  n8 = ScalarParameter() {defaultValue=0.04, parameterName=SIze}
  n9 = ScalarParameter() {defaultValue=0.6, parameterName=OffsetEnd}
  n10 = Subtract(
   A=Distance(A=TextureCoordinate(), B=ComponentMask(None=@n7) {g=true, r=true})
   B=LinearInterpolate(
    A=Multiply(A=@n8) {constB=-1}
    B=@n9
    Alpha=CollectionParameter() {collection=&DiscoverabilityMPC, parameterName=DiscoverRevealProgress}
   )
  )
  n11 = Saturate(None=Divide(A=@n10, B=@n8))
  n12 = Multiply(A=@n6.G, B=@n11)
  n13 = Saturate(None=Multiply(A=Saturate(None=@n5), B=Add(A=@n6.R, B=@n12)))
  n14 = TextureObjectParameter() {parameterName=MapTexture, texture=&Apollo_Terrain_Minimap}
  n15 = TextureCoordinate()
  n16 = ScalarParameter() {defaultValue=6, parameterName=DistortScale2}
  n17 = TextureSample(UVs=Panner(Coordinate=Multiply(A=@n15, B=@n16), Speed=ScalarParameter() {defaultValue=0.25, parameterName=DistortionSpeed})) {texture=/Game/Global/Textures/Default/Noise/3xPerlin_M}
  n18 = AppendVector(A=@n17.R, B=@n17.R)
  n19 = ComponentMask(None=VectorParameter() {parameterName=Bias2, defaultValue=<rgba 0.25 0.25 0 1>}.RGB) {g=true, r=true}
  n20 = OneMinus(None=@n11)
  n21 = Multiply(A=Multiply(A=@n6.G, B=Multiply(A=@n11, B=@n20)), B=ScalarParameter() {defaultValue=0.1, parameterName=UVdistortion2})
  n22 = Add(A=Multiply(A=Subtract(A=@n18, B=@n19) {constB=0.5}, B=@n21), B=TextureCoordinate())
  n23 = VectorParameter() {parameterName=LineColor, defaultValue=<rgba 0.05 0.05 0.05 1>}
  n24 = Divide(
   A=Add(A=ScalarParameter() {defaultValue=-0.0491, parameterName=offsetLine}, B=@n10)
   B=ScalarParameter() {defaultValue=0.005, parameterName=SIzeLine}
  )
  n25 = Saturate(None=Sine(None=Saturate(None=@n24)))
  Add(
   A=LinearInterpolate(
    A=MaterialFunctionCall!(UVsOverride, MapOpacityTexture=@n1, MapTexture=@n2) {materialFunction=&/MF_Minimap}.RGB
    B=MaterialFunctionCall!(MapTexture=@n14, UVsOverride=@n22) {materialFunction=&/MF_Minimap_FOW}.RGB
    Alpha=@n13
   )
   B=Multiply(A=@n6.G, B=Multiply(A=@n23.RGB, B=@n25))
  ) {constB=0}
  Add(A=@n21, B=ScalarParameter() {defaultValue=0.002, parameterName=UVdistortion_Ambient})
  Constant() {r=1}
  Constant() {r=1}
  Constant() {r=1}
  Frac(None=Multiply(A=Time(), B=ScalarParameter() {defaultValue=0.2, parameterName=Speed}))
  MaterialFunctionCall!(Bitmask=ScalarParameter() {defaultValue=2, parameterName=test}) {materialFunction=&/MF_DiscoverabilityPOIUnpack}
  n26 = VectorParameter() {desc=XY location and Z is Radius of the POI, parameterName=LocationRevealFX, defaultValue=<rgba 0.6788 0.6979 0.3 1>}
  Multiply(A=@n26.B, B=@n9)
  Multiply(
   A=@n13
   B=MaterialFunctionCall(
    MapOpacityTexture=TextureObjectParameter() {parameterName=MapOpacityTexture, samplerType=SAMPLERTYPE_Grayscale, texture=&T_Clyde_Minimap_PostMask}
    MapTexture=TextureObjectParameter() {parameterName=MapTexture, texture=&Apollo_Terrain_Minimap}
   ) {materialFunction=&MF_Minimap}.Alpha
  )
  n27 = TextureSample(UVs=Multiply(A=TextureCoordinate(), B=ScalarParameter() {defaultValue=2, parameterName=DistortScale1})) {texture=&3xPerlin_M}
  n28 = VectorParameter() {parameterName=Bias, defaultValue=<rgba 0.25 0.25 0 1>}
  n29 = Multiply(
   A=Subtract(A=AppendVector(A=@n27.R, B=@n27.R), B=ComponentMask(None=@n28.RGB) {g=true, r=true}) {constB=0.5}
   B=ScalarParameter() {defaultValue=0.1099, parameterName=UVdistortion}
  )
  n30 = TextureCoordinate()
  n31 = Subtract(A=Distance(A=Add(A=@n29, B=@n30), B=ComponentMask(None=@n7) {g=true, r=true}), B=@n9)
  OneMinus(None=Saturate(None=Divide(A=@n31) {constB=0.001}))
  n32 = MaterialFunctionCall(Bitmask=CollectionParameter() {collection=&DiscoverabilityMPC, parameterName=AlreadyDiscoveredBitmask}) {materialFunction=&MF_DiscoverabilityPOIUnpack}
  OneMinus(None=@n32.Result)
  Saturate(None=MaterialFunctionCall(Bitmask=CollectionParameter() {collection=&DiscoverabilityMPC, parameterName=FreshlyDiscoveredBitmask}) {materialFunction=&MF_DiscoverabilityPOIUnpack}.Result)
  n33 = Constant() {r=-1}
  Saturate(None=Subtract(A=MaterialFunctionCall(Bitmask=@n33) {materialFunction=&MF_DiscoverabilityPOIUnpack}.Result, B=@n32.Result))
  ScalarParameter() {defaultValue=-0.1, parameterName=OffsetStart}
 ]
}
