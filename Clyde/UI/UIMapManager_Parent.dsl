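; Clyde minimap manager Blueprint that swaps the map textures to match the active event week.
; On BeginPlay, where cosmetic events can run, it watches week calendar events to set WeekOffset, with Week4 deferring to nested KelpLinen events for later offsets.
; WeekOffset indexes MapTextures, feeding the minimap and mask dynamic materials from FortMiniMapInterface.
; Both materials also get a fixed opacity mask, and the mask material gets a discoverability texture.
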(event EventBeginPlay
  (Networking|CanExecuteCosmeticEvents
    (:True
      (bind _eventname (MCPContext|WatchCalendarEvents (Variables|Default|GetClydeWeekCalendarEventsToWatch))
        (:OnEventActive
          (switch string _eventname
            (:Week1
              (Variables|Default|SetWeekOffset 0))
            (:Week2
              (Variables|Default|SetWeekOffset 1))
            (:Week3
              (Variables|Default|SetWeekOffset 2))
            (:Week4
              (bind _eventname_2 (MCPContext|WatchCalendarEvents (Variables|Default|GetKelpLinenCalendarEventsToWatch))
                (:OnEventActive
                  (switch string _eventname_2
                    (:KL1
                      (Variables|Default|SetWeekOffset 3))
                    (:KL2
                      (Variables|Default|SetWeekOffset 4))
                    (:KL3
                      (Variables|Default|SetWeekOffset 5)))))))
          (Utilities|FlowControl|Sequence
            (:then_0
              (bind _returnvalue (Class|FortMiniMapInterface|GetMapMaterialMID))
              (Utilities|FlowControl|Sequence
                (:then_0
                  (Rendering|Material|SetTextureParameterValue _returnvalue "MapTexture" (Utilities|Array|Get(acopy) (Variables|Default|GetMapTextures) (Variables|Default|GetWeekOffset))))
                (:then_1
                  (Rendering|Material|SetTextureParameterValue _returnvalue "MapOpacityTexture" "/Clyde/UI/Textures/T_Clyde_Minimap_PostMask.T_Clyde_Minimap_PostMask"))))
            (:then_1
              (bind _returnvalue_2 (Class|FortMiniMapInterface|GetMapMaskMaterialMID))
              (Utilities|FlowControl|Sequence
                (:then_0
                  (Rendering|Material|SetTextureParameterValue _returnvalue_2 "MapTexture" (Utilities|Array|Get(acopy) (Variables|Default|GetMapTextures) (Variables|Default|GetWeekOffset))))
                (:then_1
                  (Rendering|Material|SetTextureParameterValue _returnvalue_2 "MapOpacityTexture" "/Clyde/UI/Textures/T_Clyde_Minimap_PostMask.T_Clyde_Minimap_PostMask"))
                (:then_2
                  (Rendering|Material|SetTextureParameterValue _returnvalue_2 "DiscoverabilityTexture" "/Clyde/UI/Textures/Apollo_Terrain_Minimap_Discoverabilty.Apollo_Terrain_Minimap_Discoverabilty"))))))))))

(fn ConstructionScript
  (|CallParentFunction))
